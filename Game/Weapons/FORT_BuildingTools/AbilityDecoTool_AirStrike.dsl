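; Deco placement tool for the Air Strike ability that previews selectable bomb impact patterns.
; BeginPlay derives power level from the owner's bomb-upgrade tags, or an onboarding override, to choose a pattern tier.
; Secondary fire cycles that tier's patterns, updating the preview and notifying the server from clients.
; PopulatePositionArrays splits out duplicate impacts for distinct particles and scales positions by spacing; HideFor temporarily hides mesh and preview on a timer.
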(event Custom|Show
  ;[1 1 1] Hides this placement tool for InTime seconds
  (CallFunction|HideMeshandPlacement false)
  (Utilities|Time|ClearandInvalidateTimerbyHandle (Variables|Default|GetHideTimerHandle)))

(event Custom|HideFor
  (params (InTime :type Float%20%28double-precision%29))
  ;[1 1 1] Hides this placement tool for InTime seconds
  (CallFunction|HideMeshandPlacement true)
  (bind _returnvalue (Utilities|Time|SetTimerbyEvent (delegate Custom|Show) InTime))
  (Variables|Default|SetHideTimerHandle _returnvalue))

(event Custom|PassInfo
  (params (Power_Level :type Integer)))

(event EventBeginPlay
  ;[0 0 0] on server
  (Variables|Default|SetPattern%20Needs%20to%20change (== (Variables|Default|GetOld%20Power%20Level) -1))
  ;;[1 0.01036 0.02031] Retrieve power level
  (bind _asfort_player_pawn (Utilities|Casting|CastToFortPlayerPawn (Game|GetInstigator)))
  (Utilities|FlowControl|Sequence
    (:then_0
      (bind _output_get (Variables|Default|SetFortPlayerPawn _asfort_player_pawn))
      (if (Variables|Default|GetOnboarding)
        (Variables|Default|SetPowerLevel 3)
        (else
          (Variables|Default|SetPowerLevel (Math|Integer|int+int (select (GameplayTags|TagContainer|CompareTagAssetInterfacetoOtherTagContainers (Variables|Default|GetTC_BombUpgrade1) (Variables|Default|GetTC_BombUpgrade3) (Variables|Default|GetTC_BombUpgrade5) (Ability|GetAbilitySystemComponent _output_get)) 1 0) (select (.Case_1%20True (GameplayTags|TagContainer|CompareTagAssetInterfacetoOtherTagContainers (Variables|Default|GetTC_BombUpgrade1) (Variables|Default|GetTC_BombUpgrade3) (Variables|Default|GetTC_BombUpgrade5) (Ability|GetAbilitySystemComponent _output_get))) 1 0) (select (.Case_2%20True (GameplayTags|TagContainer|CompareTagAssetInterfacetoOtherTagContainers (Variables|Default|GetTC_BombUpgrade1) (Variables|Default|GetTC_BombUpgrade3) (Variables|Default|GetTC_BombUpgrade5) (Ability|GetAbilitySystemComponent _output_get))) 1 0))))))
    (:then_1
      ;[0 0 0] Determine if the tool has been previously initialized - replicated
      (Variables|Default|SetPattern%20Needs%20to%20change (or (Variables|Default|GetPattern%20Needs%20to%20change) (!= (Variables|Default|GetPowerLevel) (Variables|Default|GetOld%20Power%20Level))))
      (Variables|Default|SetOld%20Power%20Level (Variables|Default|GetPowerLevel))
      ;[0 0 0] last index
      (Variables|Default|SetCurrent%20Level%20Max (Utilities|Array|LastIndex (Utilities|Struct|BreakAirStrikesPatterns (Utilities|Array|Get(acopy) (Utilities|Struct|BreakAirStrikeLeveledPatterns (Variables|Default|GetDONOTEDIT_LeveledAirStrikeArrays)) (Variables|Default|GetPowerLevel)))))
      (CallFunction|PopulatePositionArrays))))

(event DecoTool|EventOnSetPreview
  (params (FortDecoHelper :type Fort%20Deco%20Helper%20Object%20Reference))
  ;[0 0 0] Client only
  (if (Utilities|IsValid FortDecoHelper :!class /Script/Engine.KismetSystemLibrary)
    (Variables|Default|SetHelper FortDecoHelper)
    (bind _asb_placement_preview_air_strike (Utilities|Casting|CastToB_Placement_Preview_AirStrike (Variables|/Script/FortniteGame.FortDecoHelper|GetDecoPreview FortDecoHelper)))
    (Class|BPlacementPreviewAirStrike|SwitchType _asb_placement_preview_air_strike (Variables|Default|GetCurrent%20Unprocessed%20Positions%20Array) (Variables|Default|GetPosition%20Array%20Cleaned%20of%20duplicates) (Variables|Default|GetDouble%20Entries))))

(event DecoTool|EventOnPressSecondaryFire
  (params (FortDecoHelper :type Fort%20Deco%20Helper%20Object%20Reference))
  (bind _decopreview (Variables|/Script/FortniteGame.FortDecoHelper|GetDecoPreview FortDecoHelper))
  ;[0 0 0] Change Pattern
  (Variables|Default|SetCurrent%20Air%20Strike%20Sub%20Tier%20Selection (Math|Integer|%25%28Integer%29 (+ (Variables|Default|GetCurrent%20Air%20Strike%20Sub%20Tier%20Selection) 1) (+ (Variables|Default|GetCurrent%20Level%20Max) 1)))
  (CallFunction|PopulatePositionArrays)
  (Utilities|IsValid _decopreview
    (:Is%20Valid
      (bind _asb_placement_preview_air_strike (Utilities|Casting|CastToB_Placement_Preview_AirStrike _decopreview)
        (:then
          (Class|BPlacementPreviewAirStrike|SwitchType _asb_placement_preview_air_strike (Variables|Default|GetCurrent%20Unprocessed%20Positions%20Array) (Variables|Default|GetPosition%20Array%20Cleaned%20of%20duplicates) (Variables|Default|GetDouble%20Entries)))
        (:CastFailed
          _))
      (Networking|SwitchHasAuthority
        (:Remote
          (Ability|BPServerPressSecondaryFire))))
    (:Is%20Not%20Valid
      _)))

(fn HideMeshAndPlacement
  (params (InHide :type Boolean))
  (if (and (Utilities|IsValid (Variables|Visual|GetWeaponMesh) :!class /Script/Engine.KismetSystemLibrary) (Utilities|IsValid (Variables|Default|GetHelper) :!class /Script/Engine.KismetSystemLibrary))
    (Rendering|SetVisibility (Variables|Visual|GetWeaponMesh) (not InHide) true)
    (bind _asb_placement_preview_air_strike (Utilities|Casting|CastToB_Placement_Preview_AirStrike (Variables|/Script/FortniteGame.FortDecoHelper|GetDecoPreview (Variables|Default|GetHelper))))
    (Class|BPlacementPreviewAirStrike|ToggleAirStrikeParticles _asb_placement_preview_air_strike InHide)))

(fn PopulatePositionArrays
  (bind _current_unprocessed_positions_array (Variables|Default|GetCurrent%20Unprocessed%20Positions%20Array))
  (bind _position_array_cleaned_of_duplicates (Variables|Default|GetPosition%20Array%20Cleaned%20of%20duplicates))
  (bind _current_unprocessed_positions_array_2 (Variables|Default|GetCurrent%20Unprocessed%20Positions%20Array))
  ;[0 0 0] Find Layout
  (Variables|Default|SetCurrent%20Unprocessed%20Positions%20Array (Utilities|Struct|BreakStrikePatternVectorStruct (Utilities|Array|Get(acopy) (Utilities|Struct|BreakAirStrikesPatterns (Utilities|Array|Get(acopy) (Utilities|Struct|BreakAirStrikeLeveledPatterns (Variables|Default|GetDONOTEDIT_LeveledAirStrikeArrays)) (Variables|Default|GetPowerLevel))) (Variables|Default|GetCurrent%20Air%20Strike%20Sub%20Tier%20Selection))))
  (Variables|Default|SetPattern%20Needs%20to%20change false)
  (Utilities|Array|Clear (Variables|Default|GetPosition%20Array%20Cleaned%20of%20duplicates))
  (Utilities|Array|Clear (Variables|Default|GetDuplicate%20Positions))
  (Utilities|Array|Clear (Variables|Default|GetDouble%20Entries))
  ;[0 0 0] Find duplicated entries and create a clean array
  (for (_array_element _array_index) _current_unprocessed_positions_array
    (bind _returnvalue (Utilities|Array|AddUnique (Variables|Default|GetPosition%20Array%20Cleaned%20of%20duplicates) _array_element))
    (if (== _returnvalue -1)
      (Utilities|Array|Add (Variables|Default|GetDuplicate%20Positions) _array_index)))
  ;[0 0 0] Create an array for all of the duplicated indices so that they may recieve a different particle effect
  (for _array_element_2 (Variables|Default|GetDuplicate%20Positions)
    (bind (_array_element_3 _array_index_2) (Utilities|Array|ForEachLoopwithBreak _current_unprocessed_positions_array)
      (:LoopBody
        (if (Math|Vector|Equal(Vector) _array_element_3 (Utilities|Array|Get(acopy) _current_unprocessed_positions_array _array_element_2) 0.0001)
          (Utilities|Array|AddUnique (Variables|Default|GetDouble%20Entries) _array_index_2)))))
  ;[0 0 0] Scale the results
  (for (_array_element_4 _array_index_3) _current_unprocessed_positions_array_2
    (Utilities|Array|SetArrayElem _current_unprocessed_positions_array_2 _array_index_3 (* _array_element_4 (Variables|Default|GetSpacing%20Between%20the%20Impacts)) true))
  (for (_array_element_5 _array_index_4) _position_array_cleaned_of_duplicates
    (Utilities|Array|SetArrayElem _position_array_cleaned_of_duplicates _array_index_4 (* _array_element_5 (Variables|Default|GetSpacing%20Between%20the%20Impacts)) true)))

(fn ConstructionScript
  (|CallParentFunction))
