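# Animation Blueprint for the blueprint paper mesh of the building tools, adding procedural secondary motion to the paper.
# Starting from the reference pose, four chained world-space AnimDynamics nodes drive bones Spring_1 to Spring_4 at half alpha, then feed the output pose.
# Each uses damped springs with limited linear travel, locked rotation and zero gravity scale.
# Test sequence players, Spring controllers and an extra AnimDynamics node sit unconnected; the event graph is empty.
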AnimBlueprint /Game/Weapons/FORT_BuildingTools/Meshes/Blueprint_Paper_VIM
target_skeleton=&/SK_Blueprint_Paper_Skeleton
graphs [
 AnimGraph {nodes [
   Anim Dynamics - Bone: Spring_1(Alpha=1, ExternalForce="0,0,0") {
    angularSpringConstant=50
    bAngularSpring=true
    bGravityOverrideInSimSpace=true
    bLinearSpring=true
    bUseGravityOverride=true
    linearSpringConstant=50
    simulationSpace=RootRelative
    alphaBoolBlend {blendOption=Linear}
    boundBone {boneName=Spring_1}
    componentAppliedLinearAccClamp=<5000 5000 5000>
    componentLinearAccScale=<0.2 0.2 0.3>
    gravityOverride {z=-980}
    physicsBodyDefinitions [{
      sphereCollisionRadius=0
      boundBone {boneName=Spring_1}
      boxExtents=<2 2 2>
      constraintSetup {angularLimitsMax=<20 20 20>, angularLimitsMin=<-20 -20 -20>, angularTarget=<0 0 0>, linearAxesMax=<2 2 2>, linearAxesMin=<-2 -2 -2>}
      localJointOffset=<0 0 -1>
     }]
    relativeSpaceBone {boneName=Anim}
   }
   n1 = float + float(A=0, B=0.5) {is=title K2Node_CommutativeAssociativeBinaryOperator}
   n2 = Anim Dynamics - Bone: Spring_1(
    ComponentPose=Mesh Space Ref Pose!() -> (ComponentPose:Component Space Pose Link Structure) {is=title AnimGraphNode_MeshRefPose}
    Alpha=@n1
    ExternalForce="0,0,0"
   ) {
    angularDampingOverride=1
    angularSpringConstant=20000
    bAngularSpring=true
    bGravityOverrideInSimSpace=true
    bLinearSpring=true
    bOverrideAngularDamping=true
    bOverrideLinearDamping=true
    bUseGravityOverride=true
    gravityScale=0
    linearDampingOverride=1
    linearSpringConstant=115
    simulationSpace=World
    alphaBoolBlend {blendOption=Linear}
    boundBone {boneName=Spring_1}
    componentAppliedLinearAccClamp=<200 200 200>
    componentLinearAccScale=<0.2 0.2 0.3>
    componentLinearVelScale=<0.3 0.3 0.4>
    gravityOverride {z=-10}
    physicsBodyDefinitions [{
      sphereCollisionRadius=0
      boundBone {boneName=Spring_1}
      boxExtents=<10 10 10>
      constraintSetup {angularLimitsMax=<0 0 0>, angularLimitsMin=<0 0 0>, angularTarget=<0 0 0>, linearAxesMax=<10 10 10>, linearAxesMin=<-10 -10 -10>}
      localJointOffset=<0 0 0>
     }]
    relativeSpaceBone {boneName=Root}
   }
   n3 = Anim Dynamics - Bone: Spring_2(ComponentPose=@n2, Alpha=@n1, ExternalForce="0,0,0") {
    angularDampingOverride=1
    angularSpringConstant=20000
    bAngularSpring=true
    bGravityOverrideInSimSpace=true
    bLinearSpring=true
    bOverrideAngularDamping=true
    bOverrideLinearDamping=true
    bUseGravityOverride=true
    gravityScale=0
    linearDampingOverride=1
    linearSpringConstant=115
    simulationSpace=World
    alphaBoolBlend {blendOption=Linear}
    boundBone {boneName=Spring_2}
    componentAppliedLinearAccClamp=<200 200 200>
    componentLinearAccScale=<0.2 0.2 0.3>
    componentLinearVelScale=<0.3 0.3 0.4>
    gravityOverride {z=-10}
    physicsBodyDefinitions [{
      sphereCollisionRadius=0
      boundBone {boneName=Spring_2}
      boxExtents=<10 10 10>
      constraintSetup {angularLimitsMax=<0 0 0>, angularLimitsMin=<0 0 0>, angularTarget=<0 0 0>, linearAxesMax=<10 10 10>, linearAxesMin=<-10 -10 -10>}
      localJointOffset=<0 0 0>
     }]
    relativeSpaceBone {boneName=Root}
   }
   n4 = Anim Dynamics - Bone: Spring_3(ComponentPose=@n3, Alpha=@n1, ExternalForce="0,0,0") {
    angularDampingOverride=1
    angularSpringConstant=20000
    bAngularSpring=true
    bGravityOverrideInSimSpace=true
    bLinearSpring=true
    bOverrideAngularDamping=true
    bOverrideLinearDamping=true
    bUseGravityOverride=true
    gravityScale=0
    linearDampingOverride=1
    linearSpringConstant=115
    simulationSpace=World
    alphaBoolBlend {blendOption=Linear}
    boundBone {boneName=Spring_3}
    componentAppliedLinearAccClamp=<200 200 200>
    componentLinearAccScale=<0.2 0.2 0.3>
    componentLinearVelScale=<0.3 0.3 0.4>
    gravityOverride {z=-10}
    physicsBodyDefinitions [{
      sphereCollisionRadius=0
      boundBone {boneName=Spring_3}
      boxExtents=<10 10 10>
      constraintSetup {angularLimitsMax=<0 0 0>, angularLimitsMin=<0 0 0>, angularTarget=<0 0 0>, linearAxesMax=<10 10 10>, linearAxesMin=<-10 -10 -10>}
      localJointOffset=<0 0 0>
     }]
    relativeSpaceBone {boneName=Root}
   }
   n5 = Anim Dynamics - Bone: Spring_4(ComponentPose=@n4, Alpha=@n1, ExternalForce="0,0,0") {
    angularDampingOverride=1
    angularSpringConstant=20000
    bAngularSpring=true
    bGravityOverrideInSimSpace=true
    bLinearSpring=true
    bOverrideAngularDamping=true
    bOverrideLinearDamping=true
    bUseGravityOverride=true
    gravityScale=0
    linearDampingOverride=1
    linearSpringConstant=115
    simulationSpace=World
    alphaBoolBlend {blendOption=Linear}
    boundBone {boneName=Spring_4}
    componentAppliedLinearAccClamp=<200 200 200>
    componentLinearAccScale=<0.2 0.2 0.3>
    componentLinearVelScale=<0.3 0.3 0.4>
    gravityOverride {z=-10}
    physicsBodyDefinitions [{
      sphereCollisionRadius=0
      boundBone {boneName=Spring_4}
      boxExtents=<10 10 10>
      constraintSetup {angularLimitsMax=<0 0 0>, angularLimitsMin=<0 0 0>, angularTarget=<0 0 0>, linearAxesMax=<10 10 10>, linearAxesMin=<-10 -10 -10>}
      localJointOffset=<0 0 0>
     }]
    relativeSpaceBone {boneName=Root}
   }
   Output Pose(Result=Component To Local(ComponentPose=@n5)) {layerGroup=None}
   Sequence Player 'test_sequence'() {sequence=/Game/Weapons/FORT_BuildingTools/Meshes/SK_Blueprint_Paper_Skeleton/test_sequence}
   Sequence Player 'test2'() {sequence=/Game/Weapons/FORT_BuildingTools/Meshes/SK_Blueprint_Paper_Skeleton/test2}
   Sequence Player 'test'!(PlayRate:float=.5) -> (Pose) {sequence=/Game/Weapons/FORT_BuildingTools/Meshes/SK_Blueprint_Paper_Skeleton/test}
   Sequence Player 'test_open'() {sequence=/Game/Weapons/FORT_BuildingTools/Meshes/SK_Blueprint_Paper_Skeleton/test_open}
   n6 = Spring controller - Bone: Spring_1!(ComponentPose:Component Space Pose Link Structure, Alpha:float=.5) -> (Pose:Component Space Pose Link Structure) {
    is=title AnimGraphNode_SpringBone
    bLimitDisplacement=true
    maxDisplacement=40
    springDamping=10
    springStiffness=100
    alphaBoolBlend {blendOption=Linear}
    springBone {boneName=Spring_1}
   }
   n7 = Spring controller - Bone: Spring_2!(ComponentPose:Component Space Pose Link Structure=@n6, Alpha:float=.5) -> (Pose:Component Space Pose Link Structure) {
    is=title AnimGraphNode_SpringBone
    bLimitDisplacement=true
    maxDisplacement=40
    springDamping=10
    springStiffness=100
    alphaBoolBlend {blendOption=Linear}
    springBone {boneName=Spring_2}
   }
   n8 = Spring controller - Bone: Spring_3!(ComponentPose:Component Space Pose Link Structure=@n7, Alpha:float=.5) -> (Pose:Component Space Pose Link Structure) {
    is=title AnimGraphNode_SpringBone
    bLimitDisplacement=true
    maxDisplacement=40
    springDamping=10
    springStiffness=100
    alphaBoolBlend {blendOption=Linear}
    springBone {boneName=Spring_3}
   }
   Spring controller - Bone: Spring_4!(ComponentPose:Component Space Pose Link Structure=@n8, Alpha:float=.5) -> (Pose:Component Space Pose Link Structure) {
    is=title AnimGraphNode_SpringBone
    bLimitDisplacement=true
    maxDisplacement=40
    springDamping=10
    springStiffness=100
    alphaBoolBlend {blendOption=Linear}
    springBone {boneName=Spring_4}
   }
  ]}
 EventGraph {links [], nodes []}
]
variables []
